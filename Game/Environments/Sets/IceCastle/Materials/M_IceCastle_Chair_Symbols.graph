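# Surface material for symbol blocks on the Ice Castle elemental throne, showing one cell of a block sheet chosen by a parameter.
# The one-based Number parameter, minus one, becomes the frame index for the CustomNumbers_SubUV function borrowed from weapon skins, with a sub-image count of three.
# That function samples the mask, normal and diffuse sheets with identical UVs and frame, and the mask result is split into R, G and B channels.
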Material /Game/Environments/Sets/IceCastle/Materials/M_IceCastle_Chair_Symbols
graph {
 properties {bUsedWithNanite=true, pixelDepthOffsetMode=PDOM_Legacy, usedShadingModels=Default Lit}
 nodes [
  n1 = TextureCoordinate()
  n2 = Constant() {r=3}
  n3 = Add(A=ScalarParameter() {defaultValue=7, parameterName=Number}) {constB=-1}
  n4 = MaterialFunctionCall!(
   Texture=TextureObject() {samplerType=SAMPLERTYPE_Masks, texture=/Game/Environments/Sets/IceCastle/Textures/T_IceCastle_ElementalThrone_BlockSheet_S}
   UVs=@n1
   SubImages=@n2
   Frame=@n3
  ) {materialFunction=/Game/Weapons/WeaponSkins/MaterialFunction/CustomNumbers_SubUV}
  ComponentMask(None=@n4.RGB) {r=true}
  ComponentMask(None=@n4.RGB) {b=true}
  ComponentMask(None=@n4.RGB) {g=true}
  MaterialFunctionCall(
   Texture=TextureObject() {samplerType=SAMPLERTYPE_Normal, texture=/Game/Environments/Sets/IceCastle/Textures/T_IceCastle_ElementalThrone_BlockSheet_N}
   UVs=@n1
   SubImages=@n2
   Frame=@n3
  ) {materialFunction=&CustomNumbers_SubUV}
  MaterialFunctionCall(
   Texture=TextureObject() {texture=/Game/Environments/Sets/IceCastle/Textures/T_IceCastle_ElementalThrone_BlockSheet_D}
   UVs=@n1
   SubImages=@n2
   Frame=@n3
  ) {materialFunction=&CustomNumbers_SubUV}
 ]
}
